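# Surface material for the full-tile floor jump trap (bounce pad), built on material attributes with Default Lit shading.
# Diffuse, normal and tinted emissive textures form the base, with the specular mask's green channel driving metallic, specular and roughness alike.
# A HopRock gem function masked by the FX texture's red channel feeds cropped-emissive customization, which takes the base emissive as its mask.
# Cosmetic emissive eye adaptation and destructible damage functions finish the chain.
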Material /Game/Weapons/FORT_TRAPS/Trap_Jump_FullTile/Materials/M_Floor_Jump_FullTile
graph {
 properties {bAutomaticallySetUsageInEditor=false, bUseMaterialAttributes=true, usedShadingModels=Default Lit}
 nodes [
  n1 = TextureObjectParameter() {
   parameterName=HopRock Mask
   samplerType=SAMPLERTYPE_Masks
   texture=/Game/Weapons/FORT_TRAPS/Trap_Jump_FullTile/Textures/T_BouncePad_MI_Floor_Jump_FullTile_FX
  }
  n2 = TextureSampleParameter2D() {
   group=Weapon_Textures
   parameterName=Diffuse
   texture=/Game/Weapons/FORT_TRAPS/Trap_Jump_FullTile/Textures/T_BouncePad_MI_Floor_Jump_FullTile_D
  }
  n3 = TextureSampleParameter2D() {
   group=Weapon_Textures
   parameterName=SpecularMasks
   samplerType=SAMPLERTYPE_Masks
   sortPriority=1
   texture=/Game/Weapons/FORT_TRAPS/Trap_Jump_FullTile/Textures/T_BouncePad_MI_Floor_Jump_FullTile_S
  }
  n4 = TextureObjectParameter() {group=Emissive, parameterName=EmissiveTexture, texture=/Game/Weapons/FORT_TRAPS/Trap_Jump_FullTile/Textures/Floor_Jump_FullTile_E}
  n5 = TextureSampleParameter2D() {
   group=Weapon_Textures
   parameterName=Normals
   samplerType=SAMPLERTYPE_Normal
   sortPriority=2
   texture=/Game/Weapons/FORT_TRAPS/Trap_Jump_FullTile/Textures/T_BouncePad_MI_Floor_Jump_FullTile_N
  }
  n6 = MakeMaterialAttributes!(
   BaseColor=@n2.RGB
   Metallic=@n3.G
   Specular=@n3.G
   Roughness=@n3.G
   Anisotropy
   EmissiveColor=Multiply(
    A=TextureSample(Tex=@n4) {texture=&Floor_Jump_FullTile_E}.RGB
    B=VectorParameter() {group=Emissive, parameterName=EmissiveMultiplier, defaultValue=<rgba 2 2 8 1>}.RGB
   )
   Opacity
   OpacityMask
   Normal=@n5.RGB
   Tangent
   WorldPositionOffset
   SubsurfaceColor
   ClearCoat
   ClearCoatRoughness
   AmbientOcclusion
   Refraction
   CustomizedUVs_0
   CustomizedUVs_1
   CustomizedUVs_2
   CustomizedUVs_3
   CustomizedUVs_4
   CustomizedUVs_5
   CustomizedUVs_6
   CustomizedUVs_7
   PixelDepthOffset
   ShadingModel
   Displacement
  )
  n7 = MaterialFunctionCall!(
   Material=MaterialFunctionCall!(
    Alpha=TextureSample(Tex=@n1) {samplerType=SAMPLERTYPE_Masks, texture=&T_BouncePad_MI_Floor_Jump_FullTile_FX}.R
    Material=@n6
   ) {materialFunction=/Game/Weapons/WeaponSkins/MaterialFunction/HopRockGem}.Result
   Cropped Emissive Mask=GetMaterialAttributes(None=@n6) {attributeGetTypes [#1]}.Emissive Color
  ) {materialFunction=/Game/Packages/Fortress_SharedMaterials/Base_Material_Functions/BaseCustomizationFunctions/MF_Base_Customization_CroppedEmissive}
  n8 = MaterialFunctionCall!(
   Emissive
   Alpha=ScalarParameter() {defaultValue=1, group=Cosmetic Emissive EyeAdaptation, parameterName=Cosmetic Emissive EyeAdaptation Alpha}
   Material=@n7.Result
  ) {materialFunction=/Game/Packages/Fortress_SharedMaterials/Base_Material_Functions/BaseCharacterShaderFunctions/MF_CosmeticEmissiveEyeAdaptation}
  MaterialFunctionCall(Material_In=@n8.Material) {materialFunction=&MF_FN_DestructibleDamage}
 ]
}
