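# Translucent unlit particle-sprite material for the activation glow on the full-tile jump trap.
# It crops the blue channel of the trap's emissive texture and subtracts the FadeInParts dynamic parameter to form a glow mask.
# Particle color tints that mask plus a FullGlow term, feeding an emissive-adjustment function.
# A parallel sum scales the mask by particle alpha and adds FullGlow.
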Material /Game/Weapons/FORT_TRAPS/Trap_Jump_FullTile/Materials/M_TrapJump_FullTile_ActivateGlow
graph {
 properties {
  bUsedWithParticleSprites=true
  blendMode=BLEND_Translucent
  pixelDepthOffsetMode=PDOM_Legacy
  shadingModel=MSM_Unlit
  usedShadingModels=Unlit
 }
 nodes [
  n1 = VectorParameter() {parameterName=Crop, defaultValue=<rgba 0 -0.375 1.15 1>}
  n2 = TextureSample(UVs=MaterialFunctionCall!(UVs, TextureIn, UpperLeftCorner=AppendVector(A=@n1.R, B=@n1.R), LowerRightCorner=AppendVector(A=@n1.B, B=@n1.B), ExportFloat4) {materialFunction=/Engine/Functions/Engine_MaterialFunctions01/Texturing/TextureCropping}.Crop UVs) {samplerSource=SSM_Clamp_WorldGroupSettings, texture=/Game/Weapons/FORT_TRAPS/Trap_Jump_FullTile/Textures/Floor_Jump_FullTile_E}
  n3 = DynamicParameter() {defaultValue=<rgba 0 1 1 1>, paramNames [FadeInParts,FullGlow,Param3,Param4]}
  n4 = LinearInterpolate(A=Constant(), B=Constant() {r=1}, Alpha=Subtract(A=@n2.B, B=@n3.FadeInParts))
  n5 = ParticleColor()
  Add(A=Multiply(A=@n4, B=@n5.A), B=@n3.FullGlow)
  MaterialFunctionCall(Emissive=Add(A=Multiply(A=@n5.RGB, B=@n3.FullGlow), B=Multiply(A=@n5.RGB, B=@n4))) {materialFunction=&MF_FX_EmissiveAdjustments}
  Multiply(A=@n2.B, B=@n3.FadeInParts)
  Power(Base=@n2.B) {constExponent=0.5}
 ]
}
