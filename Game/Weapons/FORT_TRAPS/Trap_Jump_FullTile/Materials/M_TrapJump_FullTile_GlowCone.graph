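# Additive, unlit, two-sided mesh-particle material for the glow cone of the full-tile jump trap.
# Two panning smoke samples, with particle-driven tiling and speed, combine with a V gradient to blend dark, light and center colors into HDR emissive.
# Opacity uses vertex red, particle alpha, a view-angle fade shaped by vertex green, and depth fade, which mobile skips.
# A quality-switched refraction value derives from that opacity.
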Material /Game/Weapons/FORT_TRAPS/Trap_Jump_FullTile/Materials/M_TrapJump_FullTile_GlowCone
graph {
 properties {
  bUsedWithMeshParticles=true
  blendMode=BLEND_Additive
  pixelDepthOffsetMode=PDOM_Legacy
  refractionMethod=RM_IndexOfRefraction
  shadingModel=MSM_Unlit
  twoSided=true
  usedShadingModels=Unlit
 }
 nodes [
  n1 = VertexColor()
  n2 = Power(Base=@n1.R)
  n3 = VectorParameter() {parameterName=DarkColor, defaultValue=<rgba 0.1393 0.0863 0.2344 0>}
  n4 = VectorParameter() {parameterName=LightColor, defaultValue=<rgba 1 1 1 0>}
  n5 = ComponentMask(None=TextureCoordinate()) {g=true}
  n6 = Multiply(A=Power(Base=@n5) {constExponent=6}) {constB=0.5}
  n7 = DynamicParameter() {paramNames [YTile,PanSpeed,Param3,Param4]}
  n8 = AppendVector(A=Constant() {r=1}, B=@n7.YTile)
  n9 = Multiply(A=@n8, B=TextureCoordinate() {uTiling=3, vTiling=0.5})
  n10 = Multiply(A=Time(), B=@n7.PanSpeed)
  n11 = Multiply(A=@n8, B=TextureCoordinate() {uTiling=3, vTiling=0.2})
  n12 = LinearInterpolate(
   A=TextureSample(UVs=Panner(Coordinate=@n9, Time=@n10) {speedX=0.2, speedY=1}) {texture=&T_Inky_Smoke_Tile_Chromatic}.RGB
   B=TextureSample(UVs=Panner(Coordinate=@n11, Time=@n10) {speedX=-0.25, speedY=0.8}) {texture=&T_Inky_Smoke_Tile_Chromatic}.RGB
  )
  n13 = Multiply(A=@n5, B=Add(A=@n6, B=Multiply(A=@n12) {constB=1.5}))
  n14 = Multiply(
   A=LinearInterpolate(
    A=LinearInterpolate(A=@n3.RGB, B=@n4.RGB, Alpha=@n13)
    B=VectorParameter() {parameterName=CenterColor, defaultValue=<rgba 2 2 2 0>}.RGB
    Alpha=OneMinus(None=@n1.R)
   )
   B=ParticleColor().RGB
  )
  n15 = Multiply(A=@n2, B=@n14)
  n16 = ScalarParameter() {defaultValue=1.5, parameterName=HDRMult}
  MaterialFunctionCall(Emissive=MaterialFunctionCall(LDR Input=Clamp(None=@n15), HDR Tint=Multiply(A=@n14, B=@n16)) {materialFunction=&DeriveHDRfromLDR}.RemappedRange) {materialFunction=&MF_FX_EmissiveAdjustments}
  n17 = ComponentMask(None=TextureCoordinate()) {g=true}
  Multiply(A=@n17, B=@n17)
  n18 = Multiply(A=ParticleColor().A, B=Add(A=@n6, B=@n13))
  n19 = DotProduct(A=CameraVectorWS(), B=VertexNormalWS())
  n20 = ScalarParameter() {defaultValue=4, parameterName=DirectionalFadePower}
  n21 = VertexColor()
  n22 = Multiply(A=Power(Base=Abs(None=@n19), Exp=LinearInterpolate(B=@n20, Alpha=@n21.G) {constA=1.25}) {constExponent=4}) {constB=4}
  n23 = ScalarParameter() {defaultValue=50, parameterName=DepthFadeDist}
  n24 = Multiply(
   A=Multiply(A=Multiply(A=@n18, B=@n2), B=Clamp(None=@n22))
   B=ShadingPathSwitch(Default=DepthFade(FadeDistance=@n23), Mobile=Constant() {r=1}) {desc=Migrated from FeatureLevelSwitch}
  )
  Multiply(A=@n24, B=Power(Base=OneMinus(None=@n17)) {constExponent=4})
  n25 = Constant() {r=1}
  QualitySwitch(
   Default=LinearInterpolate(B=ScalarParameter() {defaultValue=1, parameterName=Refraction}, Alpha=Desaturation(None=@n24, Fraction=@n25)) {constA=1, constB=1.2}
   Low=Constant() {r=1}
  )
 ]
}
