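# Unlit, translucent, two-sided mesh particle material that draws a simple glow for the full-tile jump trap.
# Particle color drives emissive through the shared FX emissive adjustments function.
# Opacity combines particle alpha, an inverted Fresnel for soft silhouette edges and an inverted V-coordinate gradient, each shaped by a power.
# The default shading path adds near-camera and depth fades, while the mobile path skips both.
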Material /Game/Weapons/FORT_TRAPS/Trap_Jump_FullTile/Materials/M_TrapJump_FullTile_SimpleGlow
graph {
 properties {
  bUsedWithMeshParticles=true
  blendMode=BLEND_Translucent
  pixelDepthOffsetMode=PDOM_Legacy
  shadingModel=MSM_Unlit
  twoSided=true
  usedShadingModels=Unlit
 }
 nodes [
  n1 = ParticleColor()
  MaterialFunctionCall(Emissive=@n1.RGB) {materialFunction=&MF_FX_EmissiveAdjustments}
  n2 = OneMinus(None=Fresnel() {baseReflectFraction=0.08, exponent=0.1})
  n3 = ComponentMask(None=TextureCoordinate()) {g=true}
  n4 = Multiply(A=Multiply(A=@n1.A, B=Power(Base=@n2)), B=Power(Base=OneMinus(None=@n3)))
  n5 = MaterialFunctionCall() {materialFunction=&Near_Camera_Fade}
  n6 = ScalarParameter() {defaultValue=50, parameterName=FadeDistance}
  ShadingPathSwitch(Default=Multiply(A=Multiply(A=@n4, B=@n5.Result), B=DepthFade(FadeDistance=@n6)), Mobile=@n4) {desc=Migrated from FeatureLevelSwitch}
 ]
}
